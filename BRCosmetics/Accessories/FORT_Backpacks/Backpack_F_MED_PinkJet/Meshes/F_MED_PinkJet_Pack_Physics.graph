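# Physics asset for the Pink Jet back bling, defining capsule bodies and joints for its dynamic wings and bow.
# Each side has a two-segment wing jointed to the pack body, with wider swing at the outer phalanx.
# The bow center is angularly locked to the pack and anchors two upper loops and two three-link bottom chains.
# Tight limits restrict motion mostly to swing, with twist locked or narrowly limited.
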PhysicsAsset /BRCosmetics/Accessories/FORT_Backpacks/Backpack_F_MED_PinkJet/Meshes/F_MED_PinkJet_Pack_Physics
bodies {
 dyn_L_bow_bottom_01 [{
   shapeType=Capsule
   length=0.4
   radius=0.9491
   shapeName=dyn_L_bow_bottom_01_capsule
   center=<-1.4906 -0.1438 0.054>
   rotation=<rot -1.4067 -90.0072 -87.7043>
  }]
 dyn_L_bow_bottom_02 [{
   shapeType=Capsule
   length=0.8
   radius=1.102
   shapeName=dyn_L_bow_bottom_02_capsule
   center=<-1.6274 0 0.2042>
   rotation=<rot -0.0002 -90.0004 95.7595>
  }]
 dyn_L_bow_bottom_03 [{
   shapeType=Capsule
   length=0.8
   radius=1.1766
   shapeName=dyn_L_bow_bottom_03_capsule
   center=<-1.9906 -0.0706 0.1806>
   rotation=<rot 1.0635 90.3459 -96.1007>
  }]
 dyn_L_bow_up [{
   shapeType=Capsule
   length=0.9652
   radius=0.9299
   shapeName=dyn_L_bow_up_capsule
   center=<-0.6075 0.0792 1.6483>
   rotation=<rot 5.937 91.3756 20.4226>
  }]
 dyn_L_wing [{
   shapeType=Capsule
   length=3.2096
   radius=1.4547
   shapeName=dyn_L_wing_capsule
   center=<3.4382 -0.0834 -0.028>
   rotation=<rot 10.7529 -89.1991 -90.8137>
  }]
 dyn_L_wing_phalanx_01 [{
   shapeType=Capsule
   length=6.5675
   radius=1.3045
   shapeName=dyn_L_wing_phalanx_01_capsule
   center=<6.3746 0.3188 -2.9808>
   rotation=<rot 0 -84.996 -65.1541>
  }]
 dyn_R_bow_bottom_01 [{
   shapeType=Capsule
   length=0.4186
   radius=0.9367
   shapeName=dyn_R_bow_bottom_01_capsule
   center=<1.3849 0.1731 -0.0826>
   rotation=<rot 0.0001 86.3316 -93.6275>
  }]
 dyn_R_bow_bottom_02 [{
   shapeType=Capsule
   length=0.602
   radius=1.002
   shapeName=dyn_R_bow_bottom_02_capsule
   center=<1.5616 -0.0025 0.0225>
   rotation=<rot 6.1811 -91.2021 93.2835>
  }]
 dyn_R_bow_bottom_03 [{
   shapeType=Capsule
   length=1.228
   radius=1.0707
   shapeName=dyn_R_bow_bottom_03_capsule
   center=<1.8342 -0.1208 -0.0755>
   rotation=<rot 1.3013 84.9676 -92.7259>
  }]
 dyn_R_bow_up [{
   shapeType=Capsule
   length=0.9652
   radius=0.9299
   shapeName=dyn_L_bow_up_capsule
   center=<0.6075 -0.0792 -1.6483>
   rotation=<rot 5.937 91.3756 -159.5774>
  }]
 dyn_R_wing [{
   shapeType=Capsule
   length=3.2096
   radius=1.4547
   shapeName=dyn_L_wing_capsule
   center=<-3.4382 0.0834 0.028>
   rotation=<rot 10.7529 -89.1991 89.1863>
  }]
 dyn_R_wing_phalanx_01 [{
   shapeType=Capsule
   length=6.5675
   radius=1.3045
   shapeName=dyn_L_wing_phalanx_01_capsule
   center=<-6.3746 -0.3188 2.9808>
   rotation=<rot 0 -84.996 114.8459>
  }]
 dyn_bow [{shapeType=Capsule, length=0, radius=1.3279, shapeName=dyn_bow_capsule, center=<0.0084 0.0448 0.013>, rotation=<rot 0.0002 90.5579 179.3881>}]
 pack [{shapeType=Capsule, length=11.0534, radius=5.2923, center=<-0.355 -1.78 -10.1977>, rotation=<rot -0.0005 -89.991 -89.5814>}]
}
constraints [
 {bone1Name=dyn_L_wing, bone2Name=pack, swing1LimitDegrees=5, swing2LimitDegrees=15, twistLimitDegrees=45, twistMotion=Locked}
 {bone1Name=dyn_L_wing_phalanx_01, bone2Name=dyn_L_wing, swing1LimitDegrees=30, swing2LimitDegrees=45, twistLimitDegrees=5}
 {
  bone1Name=dyn_bow
  bone2Name=pack
  swing1LimitDegrees=45
  swing1Motion=Locked
  swing2LimitDegrees=45
  swing2Motion=Locked
  twistLimitDegrees=45
  twistMotion=Locked
 }
 {bone1Name=dyn_L_bow_bottom_01, bone2Name=dyn_bow, swing1LimitDegrees=15, swing2LimitDegrees=15, twistLimitDegrees=1}
 {bone1Name=dyn_L_bow_bottom_02, bone2Name=dyn_L_bow_bottom_01, swing1LimitDegrees=15, swing2LimitDegrees=15, twistLimitDegrees=1}
 {bone1Name=dyn_L_bow_bottom_03, bone2Name=dyn_L_bow_bottom_02, swing1LimitDegrees=15, swing2LimitDegrees=15, twistLimitDegrees=1}
 {bone1Name=dyn_L_bow_up, bone2Name=dyn_bow, swing1LimitDegrees=5, swing2LimitDegrees=5, twistLimitDegrees=45, twistMotion=Locked}
 {bone1Name=dyn_R_bow_bottom_01, bone2Name=dyn_bow, swing1LimitDegrees=15, swing2LimitDegrees=15, twistLimitDegrees=1}
 {bone1Name=dyn_R_bow_bottom_02, bone2Name=dyn_R_bow_bottom_01, swing1LimitDegrees=15, swing2LimitDegrees=15, twistLimitDegrees=1}
 {bone1Name=dyn_R_bow_bottom_03, bone2Name=dyn_R_bow_bottom_02, swing1LimitDegrees=15, swing2LimitDegrees=15, twistLimitDegrees=1}
 {bone1Name=dyn_R_wing, bone2Name=pack, swing1LimitDegrees=5, swing2LimitDegrees=15, twistLimitDegrees=0.1}
 {bone1Name=dyn_R_wing_phalanx_01, bone2Name=dyn_R_wing, swing1LimitDegrees=30, swing2LimitDegrees=45, twistLimitDegrees=5}
 {bone1Name=dyn_R_bow_up, bone2Name=dyn_bow, swing1LimitDegrees=5, swing2LimitDegrees=5, twistLimitDegrees=45, twistMotion=Locked}
]
