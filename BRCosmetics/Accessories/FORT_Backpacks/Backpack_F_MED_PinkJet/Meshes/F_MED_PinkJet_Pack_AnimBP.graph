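# Animation Blueprint for the PinkJet backpack cosmetic, adding secondary motion and state-driven poses to the pack and its wings.
# A RigidBody node simulates physics, then chained additive bone transforms offset the pack and spread the wings, blending on dive, skydive, tactical sprint and jump flags.
# The update event sets the jump flag when jumping outside skydiving and clears it when not jumping.
# It also stores a deprecated gravity-override result that the AnimGraph never reads.
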AnimBlueprint /BRCosmetics/Accessories/FORT_Backpacks/Backpack_F_MED_PinkJet/Meshes/F_MED_PinkJet_Pack_AnimBP
target_skeleton=&/F_MED_PinkJet_Pack_Skeleton
graphs [
 AnimGraph {nodes [
   n1 = Get bIsTacticalSprinting()
   n2 = Get bIsSkyDiving()
   n3 = Get bIsSkydiveDiveMode()
   n4 = RigidBody(ComponentPose=Local To Component(LocalPose=Input Pose())) {
    bClampLinearTranslationLimitToRefPose=true
    bOverrideWorldGravity=true
    overridePhysicsAsset=&/F_MED_PinkJet_Pack_Physics
    simulationSpace=BaseBoneSpace
    alphaBoolBlend {blendOption=Linear}
    baseBoneRef {boneName=root}
    componentAppliedLinearAccClamp=<3000 3000 3000>
    componentLinearAccScale=<0.5 0.5 0.5>
    componentLinearVelScale=<0.5 0.5 0.5>
    overrideWorldGravity {z=-980}
   }
   n5 = "Transform (Modify) Bone - Bone: dyn_L_wing"!(ComponentPose=@n4, bAlphaBoolEnabled:bool=True, Translation="0,0,0", Rotation="0.000,15,0.000", Scale="1,1,1") -> (Pose) {
    alphaInputType=Bool
    rotationMode=BMM_Additive
    translationMode=BMM_Additive
    translationSpace=BCS_ParentBoneSpace
    alphaBoolBlend {blendInTime=0.5, blendOption=Linear, blendOutTime=1}
    boneToModify {boneName=dyn_L_wing}
    rotation {yaw=15}
   }
   n6 = "Transform (Modify) Bone - Bone: dyn_R_wing"!(ComponentPose=@n5, bAlphaBoolEnabled:bool=True, Translation="0,0,0", Rotation="0.000,-15,-0.000", Scale="1,1,1") -> (Pose) {
    alphaInputType=Bool
    rotationMode=BMM_Additive
    translationMode=BMM_Additive
    translationSpace=BCS_ParentBoneSpace
    alphaBoolBlend {blendInTime=0.5, blendOption=Linear, blendOutTime=1}
    boneToModify {boneName=dyn_R_wing}
    rotation {yaw=-15}
   }
   n7 = "Transform (Modify) Bone - Bone: pack"!(ComponentPose=@n6, bAlphaBoolEnabled:bool=@n3, Translation="-0.000,-0.5,-2.5", Rotation="4.6,0,0", Scale) -> (Pose) {
    alphaInputType=Bool
    rotationMode=BMM_Additive
    rotationSpace=BCS_BoneSpace
    translationMode=BMM_Additive
    translationSpace=BCS_ParentBoneSpace
    alphaBoolBlend {blendInTime=1, blendOption=Linear, blendOutTime=1}
    boneToModify {boneName=pack}
    rotation {pitch=4.6}
    translation {y=-0.5, z=-2.5}
   }
   n8 = "Transform (Modify) Bone - Bone: dyn_L_wing"(ComponentPose=@n7, bAlphaBoolEnabled=@n3, Translation="0,0,0", Rotation="0.000,50,0.000", Scale="1,1,1") {
    alphaInputType=Bool
    rotationMode=BMM_Additive
    translationMode=BMM_Additive
    translationSpace=BCS_ParentBoneSpace
    alphaBoolBlend {blendInTime=0.5, blendOption=Linear, blendOutTime=1}
    boneToModify {boneName=dyn_L_wing}
    rotation {yaw=50}
   }
   n9 = "Transform (Modify) Bone - Bone: dyn_R_wing"(ComponentPose=@n8, bAlphaBoolEnabled=@n3, Translation="0,0,0", Rotation="0.000,-50,-0.000", Scale="1,1,1") {
    alphaInputType=Bool
    rotationMode=BMM_Additive
    translationMode=BMM_Additive
    translationSpace=BCS_ParentBoneSpace
    alphaBoolBlend {blendInTime=0.5, blendOption=Linear, blendOutTime=1}
    boneToModify {boneName=dyn_R_wing}
    rotation {yaw=-50}
   }
   n10 = "Transform (Modify) Bone - Bone: pack"(ComponentPose=@n9, bAlphaBoolEnabled=@n2, Translation="0,-1,-2.5", Rotation="0,0,0") {
    alphaInputType=Bool
    rotationMode=BMM_Additive
    rotationSpace=BCS_BoneSpace
    translationMode=BMM_Additive
    translationSpace=BCS_ParentBoneSpace
    alphaBoolBlend {blendInTime=1, blendOption=Linear, blendOutTime=1}
    boneToModify {boneName=pack}
    translation {y=-1, z=-2.5}
   }
   n11 = "Transform (Modify) Bone - Bone: dyn_L_wing"(ComponentPose=@n10, bAlphaBoolEnabled=@n2, Translation="0,0,0", Rotation="0.000,30,0.000", Scale="1,1,1") {
    alphaInputType=Bool
    rotationMode=BMM_Additive
    translationMode=BMM_Additive
    translationSpace=BCS_ParentBoneSpace
    alphaBoolBlend {blendInTime=0.5, blendOption=Linear, blendOutTime=1}
    boneToModify {boneName=dyn_L_wing}
    rotation {yaw=30}
   }
   n12 = "Transform (Modify) Bone - Bone: dyn_R_wing"(ComponentPose=@n11, bAlphaBoolEnabled=@n2, Translation="0,0,0", Rotation="0.000,-30,-0.000", Scale="1,1,1") {
    alphaInputType=Bool
    rotationMode=BMM_Additive
    translationMode=BMM_Additive
    translationSpace=BCS_ParentBoneSpace
    alphaBoolBlend {blendInTime=0.5, blendOption=Linear, blendOutTime=1}
    boneToModify {boneName=dyn_R_wing}
    rotation {yaw=-30}
   }
   n13 = "Transform (Modify) Bone - Bone: pack"(ComponentPose=@n12, bAlphaBoolEnabled=@n1, Translation="0.000,0,1", Rotation="0,0,0") {
    alphaInputType=Bool
    rotationMode=BMM_Additive
    rotationSpace=BCS_BoneSpace
    translationMode=BMM_Additive
    translationSpace=BCS_ParentBoneSpace
    alphaBoolBlend {blendInTime=1, blendOption=Linear, blendOutTime=1}
    boneToModify {boneName=pack}
    translation {z=1}
   }
   n14 = OR Boolean(A=Get my jump() -> (my jump:bool), B=@n1)
   n15 = "Transform (Modify) Bone - Bone: dyn_L_wing"(ComponentPose=@n13, bAlphaBoolEnabled=@n14, Translation="0,0,0", Rotation="0.000,30,0.000", Scale="1,1,1") {
    alphaInputType=Bool
    rotationMode=BMM_Additive
    translationMode=BMM_Additive
    translationSpace=BCS_ParentBoneSpace
    alphaBoolBlend {blendInTime=0.1, blendOption=Linear, blendOutTime=1}
    boneToModify {boneName=dyn_L_wing}
    rotation {yaw=30}
   }
   n16 = "Transform (Modify) Bone - Bone: dyn_R_wing"(ComponentPose=@n15, bAlphaBoolEnabled=@n14, Translation="0,0,0", Rotation="0.000,-30,-0.000", Scale="1,1,1") {
    alphaInputType=Bool
    rotationMode=BMM_Additive
    translationMode=BMM_Additive
    translationSpace=BCS_ParentBoneSpace
    alphaBoolBlend {blendInTime=0.1, blendOption=Linear, blendOutTime=1}
    boneToModify {boneName=dyn_R_wing}
    rotation {yaw=-30}
   }
   n17 = "Transform (Modify) Bone - Bone: pack"!(ComponentPose=@n16, Alpha:float, Translation="0,-1,1", Rotation="0,0,5", Scale) -> (Pose) {
    rotationMode=BMM_Additive
    rotationSpace=BCS_BoneSpace
    translationMode=BMM_Additive
    translationSpace=BCS_BoneSpace
    alphaBoolBlend {blendOption=Linear}
    boneToModify {boneName=pack}
    rotation {roll=5}
    translation {y=-1, z=1}
   }
   Output Pose(Result=Component To Local(ComponentPose=@n17)) {layerGroup=None}
   Get bIsJumping()
  ]}
 EventGraph {nodes [
   Event BlueprintUpdateAnimation()
   > n1 = Gravity Override State Function DEPRECATED(CharacterPartAnimInstance=Self-Reference(), GameTimeInSeconds=GetGameTimeInSeconds(), GravityOverrideParams=Get Gravity Override Params())
   > Set skyDiving(skyDiving:Vector=@n1) -> (Output_Get:Vector)
   > n2 = Branch(Condition=Get bIsJumping())
   > n3 = Branch(Condition=Get bIsSkyDiving())
   Set my jump(execute=@n3.else, my jump:bool=true) -> (Output_Get:bool)
   Set my jump(execute=@n2.else, my jump=false)
   TryGetPawnOwner()
  ]}
]
variables [
 Angle Wing X {type=number, default=60}
 Angle Wing Z {type=number, default=60}
 Gravity Override Params {type=GravityOverrideParamsStruct, default {skydive_GoingUp {gravityOverride=<0 -300 15000>}}}
 my jump {type=boolean}
 rotation_wing_L {type=Rotator}
 rotation_wing_R {type=Rotator}
 skyDiving {type=Vector}
 Translation_wing_L {type=Vector}
 Translation_wing_R {type=Vector}
]
